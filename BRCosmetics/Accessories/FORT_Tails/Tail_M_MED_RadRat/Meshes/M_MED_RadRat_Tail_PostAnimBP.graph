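# Post-process Anim Blueprint for the RadRat tail accessory that adjusts the tail pose and hides the tail when it would share space with a cape or long back bling.
# AnimGraph chains additive bone-space Transform Bone nodes on tail joints 1, 2 and 7 over the input pose.
# Each update stores crouched-while-stationary as float OnlyCrouch.
# On customization load, it finds the MiscOrTail mesh and sets its visibility from the back bling check.
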AnimBlueprint /BRCosmetics/Accessories/FORT_Tails/Tail_M_MED_RadRat/Meshes/M_MED_RadRat_Tail_PostAnimBP
target_skeleton=/Game/Accessories/FORT_Tails/Common/Fortnite_Base_Tail_Skeleton
graphs [
 AnimGraph {nodes [
   n1 = "Transform (Modify) Bone - Bone: C_Tail_A_1_Jnt"!(ComponentPose=Local To Component(LocalPose=Input Pose()), bAlphaBoolEnabled:bool=True, Translation="0,0,0", Rotation="30,0,0", Scale="1,1,1") -> (Pose) {
    alphaInputType=Bool
    rotationMode=BMM_Additive
    rotationSpace=BCS_BoneSpace
    translationMode=BMM_Additive
    translationSpace=BCS_BoneSpace
    alphaBoolBlend {blendInTime=1, blendOutTime=1}
    boneToModify {boneName=C_Tail_A_1_Jnt}
    rotation {pitch=30}
   }
   n2 = "Transform (Modify) Bone - Bone: C_Tail_A_1_Jnt"(ComponentPose=@n1, bAlphaBoolEnabled=True, Translation="0,0,0", Rotation="60,0,0", Scale="1,1,1") {
    alphaInputType=Bool
    rotationMode=BMM_Additive
    rotationSpace=BCS_BoneSpace
    translationMode=BMM_Additive
    translationSpace=BCS_BoneSpace
    alphaBoolBlend {blendInTime=0.5, blendOutTime=0.8}
    boneToModify {boneName=C_Tail_A_1_Jnt}
    rotation {pitch=60}
   }
   n3 = "Transform (Modify) Bone - Bone: C_Tail_A_2_Jnt"!(ComponentPose=@n2, bAlphaBoolEnabled:bool=True, Translation="0,0,10", Rotation="60,0,0", Scale="1,1,1") -> (Pose) {
    alphaInputType=Bool
    rotationMode=BMM_Additive
    rotationSpace=BCS_BoneSpace
    translationMode=BMM_Additive
    translationSpace=BCS_BoneSpace
    alphaBoolBlend {blendInTime=0.5, blendOutTime=0.8}
    boneToModify {boneName=C_Tail_A_2_Jnt}
    rotation {pitch=60}
    translation {z=10}
   }
   n4 = "Transform (Modify) Bone - Bone: C_Tail_A_1_Jnt"!(ComponentPose=@n3, Alpha:float=1, Translation="0,0,0", Rotation="20,0,0", Scale="1,1,1") -> (Pose) {
    rotationMode=BMM_Additive
    rotationSpace=BCS_BoneSpace
    translationMode=BMM_Additive
    translationSpace=BCS_BoneSpace
    alphaBoolBlend {blendInTime=1, blendOutTime=1}
    alphaScaleBiasClamp {bInterpResult=true}
    boneToModify {boneName=C_Tail_A_1_Jnt}
    rotation {pitch=20}
   }
   n5 = "Transform (Modify) Bone - Bone: C_Tail_A_7_Jnt"(ComponentPose=@n4, Alpha=1, Translation="0,0,0", Rotation="30,0,0", Scale="1,1,1") {
    rotationMode=BMM_Additive
    rotationSpace=BCS_BoneSpace
    translationMode=BMM_Additive
    translationSpace=BCS_BoneSpace
    alphaBoolBlend {blendInTime=1, blendOutTime=1}
    alphaScaleBiasClamp {bInterpResult=true}
    boneToModify {boneName=C_Tail_A_7_Jnt}
    rotation {pitch=30}
   }
   Output Pose(Result=Component To Local(ComponentPose=@n5))
  ]}
 EventGraph {nodes [
   Event BlueprintUpdateAnimation()
   n1 = NOT Boolean(A=Get bIsMoving2D())
   n2 = Get bIsCrouching()
   > Set OnlyCrouch(OnlyCrouch:double="To Float (Boolean)"(InBool=AND Boolean(A=@n2, B=@n1))) -> (Output_Get:double)
   n3 = Event OnCharacterCustomizationLoaded()
   > n4 = Set Output(Output:Skeletal Mesh Component Budgeted Object Reference=Get Array Item with EFort Custom Part Type(Array=@n3.CustomMeshParts, PartType=MiscOrTail)) -> (Output_Get:Skeletal Mesh Component Budgeted Object Reference)
   n5 = CheckForBackbling()
   > n6 = Set isWearingCape(isWearingCape=OR Boolean(A=@n5.bIsBackblingACape, B=@n5.bIsBackblingLong))
   > n7 = Branch(Condition=@n6)
   > SetVisibility!(execute, self:Scene Component Object Reference=@n4, bNewVisibility:bool=false, bPropagateToChildren:bool=false) -> (then)
   SetVisibility(execute=@n7.else, self=@n4, bNewVisibility=true, bPropagateToChildren=false)
   TryGetPawnOwner()
  ]}
]
variables [
 isWearingCape {type=boolean}
 OnlyCrouch {type=number}
 Output {type=SkeletalMeshComponentBudgeted}
]
